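# Translucent, unlit, two-sided Niagara flame material for the Cactus Rocker alternate male outfit.
# Emissive layers scrolling plasma, smoke and randomized highlights over a curve-atlas gradient, tinted by time of day.
# Opacity shapes distorted panning noise with edge, bottom and split masks, then applies depth, view-angle, particle-alpha, time-of-day and trail-radius fades.
# World position offset combines Niagara-driven clamped bending, noise waviness and an optional frontend-only local offset.
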Material /Game/Characters/Player/Male/Medium/Bodies/M_MED_Cactus_Rocker/FX/M_Cactus_Rocker_Alt_Male_Flame
graph {
 properties {
  bUsedWithNiagaraMeshParticles=true
  bUsedWithNiagaraRibbons=true
  bUsedWithNiagaraSprites=true
  blendMode=BLEND_Translucent
  pixelDepthOffsetMode=PDOM_Legacy
  shadingModel=MSM_Unlit
  twoSided=true
  usedShadingModels=Unlit
 }
 nodes [
  n1 = Constant()
  n2 = DynamicParameter() {desc=Parameters set by Niagara, parameterIndex=1, paramNames [X,Y,Z,WPOIntensity]}
  n3 = MaterialFunctionCall(X=@n2.X, Y=@n2.X, Z=@n2.X) {materialFunction=&MakeFloat3}
  n4 = ParticlePositionWS()
  n5 = ComponentMask(None=OneMinus(None=TextureCoordinate())) {g=true}
  n6 = ScalarParameter() {defaultValue=3, parameterName=WPOBendPower}
  n7 = LinearInterpolate(
   B=Clamp(
    None=Subtract(A=@n3.Result, B=@n4)
    Min=ScalarParameter() {defaultValue=-100, parameterName=Clamp_min}
    Max=ScalarParameter() {defaultValue=100, parameterName=Clamp_max}
   ) {maxDefault=1000, minDefault=-1000}
   Alpha=Multiply(A=Power(Base=@n5, Exp=@n6), B=@n2.WPOIntensity)
  )
  n8 = ScalarParameter() {defaultValue=1, parameterName=Use interpolate WPO}
  n9 = Panner(
   Coordinate=Add(
    A=DynamicParameter() {defaultValue=<rgba 1 1 1 50>, paramNames [Offset, Texture_mask_offset, WaveIntensity, Trail_Rang]}.Offset
    B=TextureCoordinate() {coordinateIndex=1}
   )
   Time=Multiply(A=Time(), B=ScalarParameter() {defaultValue=1.1, group=Panner Speed, parameterName=Speed})
  ) {speedY=0.8}
  n10 = TextureSample(UVs=@n9) {samplerType=SAMPLERTYPE_Normal, texture=/Game/Characters/Player/Female/Medium/Bodies/F_MED_Bistro_Astronaut/FX/T_GaussianNoise_Normal}
  n11 = ComponentMask(None=@n10.RGB) {g=true, r=true}
  n12 = TextureSample(UVs=@n9) {texture=&T_Smooth_tile}
  n13 = ComponentMask(None=TextureCoordinate()) {g=true}
  n14 = Multiply(
   A=LinearInterpolate(B=AppendVector(A=@n11, B=@n12.R), Alpha=OneMinus(None=@n13))
   B=DynamicParameter() {defaultValue=<rgba 1 1 1 50>, paramNames [Offset, Texture_mask_offset, WaveIntensity, Trail_Rang]}.WaveIntensity
  ) {constB=10}
  n15 = ScalarParameter() {desc=no_Wpo, group=Simple_Round_particle, parameterName=No_WPO}
  n16 = TextureCoordinate()
  n17 = Multiply(
   A=VectorParameter() {parameterName=Flame offset in frontend, defaultValue=<rgba 0 -20 0 1>}.RGB
   B=OneMinus(None=ComponentMask(None=@n16) {g=true})
  )
  Add(
   A=Add(A=LinearInterpolate(A=@n1, B=@n7, Alpha=@n8), B=LinearInterpolate(A=@n14, Alpha=@n15))
   B=LinearInterpolate(
    A=Constant()
    B=Transform(None=@n17) {transformSourceType=TRANSFORMSOURCE_Local}
    Alpha=DynamicParameter() {parameterIndex=2, paramNames [Frontend offset fixed, none, none, none]}.Frontend offset fixed
   )
  ) {constB=0}
  n18 = VectorParameter() {parameterName=Color_Morning, defaultValue=<rgba 0.368 0.1625 0.4271 1>}
  n19 = Constant() {r=1}
  n20 = VectorParameter() {parameterName=Color_Evening, defaultValue=<rgba 0.3054 0.1556 0.3385 1>}
  n21 = VectorParameter() {parameterName=Color_Night, defaultValue=<rgba 0.6152 0.625 0.6055 0.7>}
  n22 = TextureSample!(UVs=Panner() {speedY=0.5}, Tex, Bias=Constant() {r=2}, Apply View MipBias) {mipValueMode=TMVM_MipBias, texture=/Game/Effects/Fort_Effects/Textures/Plasma/T_Plasma_Aurora_Seamless}
  n23 = ParticleColor()
  n24 = CurveAtlasRowParameter(CurveTime=ComponentMask(None=TextureCoordinate()) {g=true}) {
   atlas=/Game/Athena/LinearColorCurves/A_Atlas_Turtleneck_Glider
   curve=&/Cactus_Rocker_Male_FlameCurve
   defaultValue=7
   parameterName=ColorCurve
  }
  n25 = VectorParameter() {group=Emissive, parameterName=Dark_Color, defaultValue=<rgba 0.2314 0.1848 1 1>}
  n26 = VectorParameter() {group=Emissive, parameterName=Light_Color, defaultValue=<rgba 1 1 1 1>}
  n27 = TextureSample!(UVs=Panner() {speedY=1}, Tex, Apply View MipBias) {texture=&T_Inky_Smoke_Tile}
  n28 = Add(
   A=Multiply(A=Power(Base=@n22.RGB)) {constB=3}
   B=Multiply(A=Multiply(A=@n23.RGB, B=@n24), B=LinearInterpolate(A=@n25.RGB, B=@n26.RGB, Alpha=@n27.R))
  )
  n29 = TextureCoordinate()
  n30 = ComponentMask(None=@n29) {g=true}
  n31 = ParticleRandom()
  n32 = Panner(Coordinate=AppendVector(A=ComponentMask(None=@n29) {r=true}, B=Add(A=@n30, B=@n31))) {speedY=1}
  n33 = Multiply(
   A=VectorParameter() {group=Emissive, parameterName=Highlight_Color, defaultValue=<rgba 0 131.2977 150 1>}.RGB
   B=Power(Base=TextureSample(UVs=@n32) {texture=&T_Inky_Smoke_Tile}.R) {constExponent=10}
  )
  MaterialFunctionCall(Emissive=Multiply(
    A=Add(A=@n28, B=@n33)
    B=MaterialFunctionCall(Color_Morning=@n18.RGB, Color_Day=@n19, Color_Evening=@n20.RGB, Color_Night=@n21.RGB) {materialFunction=&Color_Time_of_Day}.Result
   )) {materialFunction=&MF_FX_EmissiveAdjustments}
  n34 = MaterialFunctionCall("FadeSensitivity (optional)"=ScalarParameter() {group=Opacity, parameterName=FadeBasedOnViewAngle}) {materialFunction=&FadeBasedOnViewAngle}
  n35 = Add(
   A=DynamicParameter() {defaultValue=<rgba 1 1 1 50>, paramNames [Offset, Texture_mask_offset, WaveIntensity, Trail_Rang]}.Offset
   B=TextureCoordinate() {coordinateIndex=1, vTiling=1.5}
  )
  n36 = ComponentMask(None=VectorParameter() {group=Distortion, parameterName=Distortion UV Scale, sortPriority=3, defaultValue=<rgba 1 1 0 0>}.RGB) {g=true, r=true}
  n37 = Time()
  n38 = ScalarParameter() {defaultValue=1, group=Panner Speed, parameterName=Speed}
  n39 = TextureSampleParameter2D(UVs=Panner(Coordinate=Multiply(A=@n35, B=@n36), Time=Multiply(A=@n37, B=@n38)) {speedY=1}) {
   group=Distortion
   parameterName=Distortion Normal
   samplerType=SAMPLERTYPE_Normal
   sortPriority=2
   texture=&T_Weap_Distort_N_02
  }
  n40 = ComponentMask(None=@n39.RGB) {g=true, r=true}
  n41 = TextureCoordinate()
  n42 = VectorParameter() {group=Fire Shape, parameterName=Fire_MacroShape_UVScale, sortPriority=5, defaultValue=<rgba 1 1 0 0>}
  n43 = ComponentMask(None=@n42.RGB) {g=true, r=true}
  n44 = ComponentMask(None=Multiply(A=@n41, B=@n43)) {g=true}
  n45 = LinearInterpolate(
   B=Divide(
    A=LinearInterpolate(A=@n40, Alpha=@n44) {constB=0}
    B=ScalarParameter() {defaultValue=3, group=Distortion, parameterName=UV_Distortion_Divsor, sortPriority=1}
   ) {constB=20}
   Alpha=ScalarParameter() {defaultValue=1, group=Distortion, parameterName=UV_Distortion_Lerp}
  )
  n46 = Add(
   A=DynamicParameter() {defaultValue=<rgba 1 1 1 50>, paramNames [Offset, Texture_mask_offset, WaveIntensity, Trail_Rang]}.Offset
   B=TextureCoordinate() {coordinateIndex=1, uTiling=0.8, vTiling=0.9}
  )
  n47 = ComponentMask(None=VectorParameter() {group=Fire Shape, parameterName=Fire_Micro_UVScale, sortPriority=8, defaultValue=<rgba 1 1 0 0>}.RGB) {g=true, r=true}
  n48 = Panner(
   Coordinate=MaterialFunctionCall(XY=Multiply(A=@n46, B=@n47)) {materialFunction=&Swizzle}.YX
   Time=Multiply(A=Time(), B=ScalarParameter() {defaultValue=1, group=Panner Speed, parameterName=Speed})
  ) {speedX=1.2, speedY=0.2}
  n49 = TextureSampleParameter2D(UVs=ShadingPathSwitch(Default=Add(A=@n45, B=@n48), Mobile=@n48) {desc=Migrated from FeatureLevelSwitch}) {
   group=Fire Shape
   parameterName=Fire Micro Texture
   samplerType=SAMPLERTYPE_LinearColor
   sortPriority=6
   texture=&T_NoisePack_LinearColor
  }
  n50 = TextureSample() {samplerType=SAMPLERTYPE_LinearGrayscale, texture=/Engine/Functions/Engine_MaterialFunctions02/ExampleContent/Textures/LowResBlurredNoise}
  n51 = MaterialFunctionCall(Float2=Multiply(A=@n41, B=@n43)) {materialFunction=&BreakOutFloat2Components}
  n52 = Power(
   Base=Sine(None=Multiply(A=@n51.R) {constB=0.5})
   Exp=LinearInterpolate(
    A=ScalarParameter() {defaultValue=6, group=Fire Shape, parameterName=Edge Fade Value Top, sortPriority=3}
    B=ScalarParameter() {defaultValue=2, group=Fire Shape, parameterName=Edge Fade Value Bottom, sortPriority=4}
    Alpha=@n51.G
   ) {constA=6, constB=2}
  ) {constExponent=4}
  n53 = VectorParameter() {group=BottomFade, parameterName=Bottom Fade_SphereMask_A_Scale, sortPriority=2, defaultValue=<rgba 1.2 1 0 0>}
  n54 = Multiply(A=TextureCoordinate(), B=ComponentMask(None=@n53.RGB) {g=true, r=true})
  n55 = VectorParameter() {group=BottomFade, parameterName=Bottom Fade_SphereMask_B, sortPriority=3, defaultValue=<rgba 0.6 0.05 0 0>}
  n56 = ComponentMask(None=@n55.RGB) {g=true, r=true}
  n57 = ScalarParameter() {defaultValue=0.95, group=BottomFade, parameterName=BottomFade_Radius, sortPriority=1}
  n58 = Saturate(None=Multiply(A=SphereMask(A=@n54, B=@n56, Radius=@n57) {attenuationRadius=0.95, hardnessPercent=0}) {constB=15})
  n59 = ScalarParameter() {defaultValue=1, group=BottomFade, parameterName=Bottom Fade}
  n60 = Multiply(
   A=LinearInterpolate(
    A=ChannelMaskParameter(None=@n49.RGBA) {group=Fire Shape, parameterName=Fire Micro Mask, sortPriority=7}.R
    B=@n50.R
    Alpha=ScalarParameter() {defaultValue=0.2, parameterName=Fire Micro lerp}
   ) {constAlpha=0.2}
   B=Multiply(A=Multiply(A=@n52, B=@n51.G), B=LinearInterpolate(B=@n58, Alpha=@n59) {constA=1})
  )
  n61 = LinearInterpolate(Alpha=@n51.G) {constA=2}
  n62 = ScalarParameter() {defaultValue=2, group=Fire Shape, parameterName=Fire Split Contrast, sortPriority=1}
  n63 = OneMinus(None=Sine(None=Multiply(A=@n51.R) {constB=0.45}))
  n64 = LinearInterpolate(Alpha=@n39.R) {constA=0.5, constB=1.5}
  n65 = Round(None=Add(A=Power(Base=@n51.G, Exp=@n62), B=Power(Base=@n63, Exp=@n64) {constExponent=0.5}))
  n66 = ScalarParameter() {defaultValue=1, group=Fire Shape, parameterName=Fire Split}
  n67 = Multiply(
   A=Multiply(A=Power(Base=@n60, Exp=@n61), B=LinearInterpolate(B=@n65, Alpha=@n66) {constA=1, constB=0})
   B=ScalarParameter() {defaultValue=15, group=Opacity, parameterName=OpacityMaskMultiplier, sortPriority=1}
  ) {constB=3}
  n68 = LinearInterpolate(A=Floor(None=@n67), B=@n67, Alpha=ScalarParameter() {defaultValue=1, parameterName=Floor})
  n69 = DepthFade(Opacity=@n68, FadeDistance=ScalarParameter() {defaultValue=200, group=Opacity, parameterName=DepthFadeDistance})
  n70 = Multiply(
   A=ShadingPathSwitch(Default=Multiply(A=@n69), Mobile=@n68) {desc=Migrated from FeatureLevelSwitch}
   B=LinearInterpolate(B=@n23.A, Alpha=ScalarParameter() {defaultValue=1, parameterName=Multiply by Particle Alpha}) {constA=1}
  )
  n71 = WorldPosition()
  n72 = ParticlePositionWS()
  n73 = DynamicParameter() {defaultValue=<rgba 1 1 1 50>, paramNames [Offset, Texture_mask_offset, WaveIntensity, Trail_Rang]}
  Multiply(
   A=MaterialFunctionCall(Value_Morning=@n18.A, Value_Day=@n19, Value_Evening=@n20.A, Value_Night=@n21.A) {materialFunction=&Scalar_Time_of_Day}.Result
   B=Multiply(A=Multiply(A=@n70, B=@n34.Result), B=SphereMask(A=@n71.XYZ, B=@n72, Radius=@n73.Trail_Rang) {attenuationRadius=50, hardnessPercent=0.2})
  )
  ParticleRandom()
  ScalarParameter() {defaultValue=1, parameterName=Frontend_Flame_Offset_Fixed}
 ]
}
